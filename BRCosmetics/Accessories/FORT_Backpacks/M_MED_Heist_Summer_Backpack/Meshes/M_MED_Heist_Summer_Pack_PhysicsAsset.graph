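# Physics asset for the M_MED_Heist_Summer backpack cosmetic's skeletal mesh.
# It holds a single body on the root bone, shaped as one tilted capsule offset from the bone to approximate the pack.
# There are no constraints, so the asset defines no jointed or ragdoll motion between parts.
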PhysicsAsset /BRCosmetics/Accessories/FORT_Backpacks/M_MED_Heist_Summer_Backpack/Meshes/M_MED_Heist_Summer_Pack_PhysicsAsset
bodies {root [{shapeType=Capsule, length=16.1916, radius=8.0449, center=<0.0004 -7.4716 0.562>, rotation=<rot 7.3315 -90 -0.0031>}]}
constraints []
